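# M_DurrBox is a Default Lit material for the DurrBox, part of the Durrrables Burger consumable in Athena.
# It samples diffuse, normal and packed S textures for the box.
# The red and blue channels of the S texture each drive a Lerp that remaps the mask into a narrower constant range.
# The graph does not show which material inputs these nodes feed.
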Material /Game/Athena/Items/Consumables/DurrrablesBurger/M_DurrBox
graph {
 properties {pixelDepthOffsetMode=PDOM_Legacy, usedShadingModels=Default Lit}
 nodes [
  n1 = TextureSample() {texture=&/T_DurrBox_S}
  LinearInterpolate(Alpha=@n1.B) {constA=0.2, constB=0.9}
  LinearInterpolate(Alpha=@n1.R) {constA=0.1, constB=0.6}
  TextureSample() {texture=&/T_DurrBox_D}
  TextureSample() {samplerType=SAMPLERTYPE_Normal, texture=&/T_Durrbox_N}
 ]
}
