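# Subsurface-shaded material for the Durrr Burger consumable pickup, with switchable sweep, subsurface and liquid features.
# The default-on sweep projects pivot-relative world position onto a camera-space direction, then pans and wraps it into a narrow moving band of coloured glow.
# Optional subsurface combines Durrr drink masks, Fresnel and diffuse luminance into tinted colour and distance-blended glow.
# Optional liquid uses a hot-sauce mask to blend a Fresnel-driven two-colour gradient over the diffuse.
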Material /Game/Athena/Items/Consumables/DurrrablesBurger/M_DurrBurgerPickup_NoNormal
graph {
 properties {pixelDepthOffsetMode=PDOM_Legacy, shadingModel=MSM_Subsurface, usedShadingModels=Subsurface}
 nodes [
  n1 = Multiply(
   A=Subtract(
    A=WorldPosition().XYZ
    B=MaterialFunctionCall() {materialFunction=&ComponentPivotLocation}.Result
   )
   B=ScalarParameter() {defaultValue=0.004, parameterName=WorldTileScale}
  )
  n2 = Normalize(VectorInput=Transform(None=VectorParameter() {parameterName=SweepDirectionCamSpace, defaultValue=<rgba -0.5 0.45 -1 0>}.RGB) {transformSourceType=TRANSFORMSOURCE_Camera})
  n3 = ComponentMask(None=Multiply(A=DotProduct(A=@n1, B=@n2), B=@n2)) {b=true, r=true}
  n4 = ScalarParameter() {defaultValue=0.18, parameterName=PanSpeed}
  n5 = Frac(None=ComponentMask(None=Panner(Coordinate=@n3, Speed=@n4) {speedX=0.2, speedY=0.2}) {g=true})
  n6 = Multiply(A=Subtract(A=@n5) {constB=0.97}) {constB=33.3333}
  n7 = Multiply(A=Floor(None=Add(A=@n5) {constB=0.03}), B=Power(Base=OneMinus(None=@n6)) {constExponent=3})
  n8 = VectorParameter() {parameterName=SweepColour, defaultValue=<rgba 2 1.6277 0.6 0>}
  n9 = TextureSampleParameter2D() {parameterName=Subsurface Masks, texture=/Game/Athena/Items/Consumables/DurrrablesDrink/T_DurrDrinkPickup_O}
  n10 = MaterialFunctionCall(Exponent=Constant() {r=2}) {materialFunction=&MF_Fresnel_FX}
  n11 = Multiply(A=DotProduct(
    A=TextureSampleParameter2D() {parameterName=Diffuse, texture=&/T_DurrBurgerPickup_D}.RGB
    B=Constant3Vector() {constant=<rgba 0.3 0.59 0.11 0>}
   ))
  n12 = Add(
   A=Multiply(A=@n9.G, B=ScalarParameter() {defaultValue=2, parameterName=ShakeBrightness})
   B=Multiply(A=Multiply(A=@n9.B, B=@n10.Fresnel), B=Power(Base=@n11) {constExponent=3})
  )
  n13 = ScalarParameter() {defaultValue=1024, parameterName=DistanceEmissionBlendRange}
  n14 = ScalarParameter() {defaultValue=256, parameterName=DistanceEmissionBlendOffset}
  n15 = Multiply(
   A=VectorParameter() {parameterName=SubsurfaceColour, defaultValue=<rgba 0.0434 0.442 1 0>}.RGB
   B=OneMinus(None=MaterialFunctionCall!(Blend Range=@n13, Start Offset=@n14) {materialFunction=/Engine/Functions/Engine_MaterialFunctions03/Particles/Distance_Blend}.Result)
  ) {constB=0}
  Add(
   A=StaticSwitchParameter(True=Multiply(A=@n7, B=@n8.RGB), False=Constant()) {defaultValue=true, parameterName=UseSweep}
   B=StaticSwitchParameter(True=Multiply(A=@n12, B=@n15), False=Constant()) {parameterName=UseSubsurface}
  )
  n16 = TextureSampleParameter2D() {parameterName=Subsurface Masks, texture=&T_DurrDrinkPickup_O}
  n17 = Constant() {r=1}
  n18 = Constant() {r=1.5}
  n19 = Multiply(A=Add(A=MaterialFunctionCall(Brightness=@n17, Exponent=@n18) {materialFunction=&MF_Fresnel_FX}.Fresnel Inverted) {constB=0.15}) {constB=0.3}
  n20 = Saturate(None=Add(A=@n16.G, B=Multiply(A=@n16.R, B=@n19)))
  n21 = VectorParameter() {parameterName=SubsurfaceColour, defaultValue=<rgba 0.0434 0.442 1 0>}
  n22 = OneMinus(None=TextureSampleParameter2D() {parameterName=Subsurface Masks, texture=&T_DurrDrinkPickup_O}.R)
  n23 = TextureSampleParameter2D() {parameterName=Diffuse, texture=&T_DurrBurgerPickup_D}
  StaticSwitchParameter(True=Add(A=Multiply(A=@n20, B=@n21.RGB), B=Multiply(A=@n22, B=@n23.RGB)), False=Constant()) {parameterName=UseSubsurface}
  n24 = TextureSampleParameter2D() {parameterName=Diffuse, texture=&T_DurrBurgerPickup_D}
  n25 = VectorParameter() {parameterName=LiquidBaseColor}
  n26 = VectorParameter() {parameterName=LiquidFresnelColor}
  n27 = MaterialFunctionCall(Exponent=ScalarParameter() {defaultValue=1, parameterName=LiquidFresnelExponent}) {materialFunction=&MF_Fresnel_FX}
  StaticSwitchParameter(
   True=LinearInterpolate(
    A=@n24.RGB
    B=LinearInterpolate(A=@n25.RGB, B=@n26.RGB, Alpha=@n27.Fresnel Inverted)
    Alpha=TextureSample() {samplerType=SAMPLERTYPE_Masks, texture=/Game/Athena/Items/Consumables/DurrrablesFries/T_DurrHotsauceLiquidMask}.R
   )
   False=@n24.RGB
  ) {parameterName=UseLiquid}
  TextureSampleParameter2D() {parameterName=SpecularMasks, texture=&/T_DurrBurgerPickup_S}
  TextureSampleParameter2D() {parameterName=Normal, samplerType=SAMPLERTYPE_Normal, texture=&/T_DurrBurgerPickup_N}
 ]
}
